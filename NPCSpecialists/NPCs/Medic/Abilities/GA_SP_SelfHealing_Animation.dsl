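; Gameplay ability that makes the Medic NPC specialist bot heal itself with a Chug Splash when health or shield is low.
; On activation it applies a SetByCaller cooldown, lengthened by the pawn's count of consecutive self-heals in the storm.
; A recursive Heal event equips the item and pulses secondary fire up to a limit, then increments or resets that storm count.
; On end, the bot re-equips its best weapon.
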(event Ability|EventActivateAbility
  (bind _returnvalue (Ability|MakeOutgoingGameplayEffectSpec "/NPCSpecialists/NPCs/Medic/Abilities/GE_SelfHealCooldown.GE_SelfHealCooldown_C"))
  (bind _selfhealcooldown_hf (Variables|Default|GetSelfHealCooldown_Hf))
  (bind _returnvalue_2 (ScalableFloat|GetValueatLevel _selfhealcooldown_hf 1.0))
  (Ability|CommitAbility)
  (bind _asbp_player_pawn_sp_medic (Utilities|Casting|CastToBP_PlayerPawn_SP_Medic (Ability|GetAvatarActorfromActorInfo))
    (:then
      (Variables|InternalData|SetAIPawn _asbp_player_pawn_sp_medic)
      (bind _returnvalue_3 (Ability|GameplayEffect|AssignTagSetbyCallerMagnitude _returnvalue "(TagName=\"SetByCaller.Cooldown\")" (+ _returnvalue_2 (* (ScalableFloat|GetValueatLevel (Variables|Default|GetHealAdditionalStormCd_Hf) 1.0) (Variables|/NPCSpecialists/NPCs/Medic/Pawn/BP_PlayerPawn_SP_Medic.BP_PlayerPawn_SP_Medic_C|GetSelfHealsInStorm (Variables|InternalData|GetAIPawn))))))
      (Ability|ApplyGameplayEffectSpecToOwner _returnvalue_3)
      (bind _asfort_player_pawn (Utilities|Casting|CastToFortPlayerPawn (Ability|GetAvatarActorfromActorInfo))
        (:then
          (Variables|InternalData|SetBotPawn _asfort_player_pawn)
          (bind _asfort_athena_aibot_controller (Utilities|Casting|CastToFortAthenaAIBotController (AI|GetAIController (Ability|GetAvatarActorfromActorInfo)))
            (:then
              (Variables|InternalData|SetBotController _asfort_athena_aibot_controller)
              (Variables|InternalData|SetHealCounter 0)
              (CallFunction|Heal))
            (:CastFailed
              (Ability|EndAbility))))
        (:CastFailed
          (Ability|EndAbility))))
    (:CastFailed
      (bind _returnvalue_4 (Ability|GameplayEffect|AssignTagSetbyCallerMagnitude _returnvalue "(TagName=\"SetByCaller.Cooldown\")" _returnvalue_2))
      (Ability|ApplyGameplayEffectSpecToOwner _returnvalue_4)
      (Ability|EndAbility))))

(event Custom|Heal
  (bind _aipawn (Variables|InternalData|GetAIPawn))
  (bind _returnvalue (Class|FortAthenaAIBotController|EquipWeaponbyTag (Variables|InternalData|GetBotController) (Variables|Default|GetChugSplashWeaponTag)))
  (if (Utilities|IsValid _returnvalue :!class /Script/Engine.KismetSystemLibrary)
    (Utilities|FlowControl|Delay (ScalableFloat|GetValueatLevel (Variables|Default|GetDelayBeforeHealingAgain)))
    (Class|FortAthenaAIBotController|StartSecondaryFire (Variables|InternalData|GetBotController))
    (Utilities|FlowControl|Delay 0.75)
    (Class|FortAthenaAIBotController|StopSecondaryFire (Variables|InternalData|GetBotController) true)
    (Variables|InternalData|SetHealCounter (+ (Variables|InternalData|GetHealCounter) 1))
    (if (>= (Variables|InternalData|GetHealCounter) (ScalableFloat|GetValueasInteger (Variables|Default|GetHealCountLimit)))
      (Utilities|FlowControl|Delay 0.3)
      (if (GameplayTags|HasMatchingGameplayTag (Variables|InternalData|GetAIPawn) (Variables|Default|GetOutsideSafeZoneTag))
        (Math|Integer|IncrementInt (Variables|/NPCSpecialists/NPCs/Medic/Pawn/BP_PlayerPawn_SP_Medic.BP_PlayerPawn_SP_Medic_C|GetSelfHealsInStorm _aipawn)
          (:%20%20
            _))
        (else
          (Variables|/NPCSpecialists/NPCs/Medic/Pawn/BP_PlayerPawn_SP_Medic.BP_PlayerPawn_SP_Medic_C|SetSelfHealsInStorm 0 _aipawn)))
      ;[0 1 0.225] End Ability
      (Ability|EndAbility)
      (else
        (CallFunction|Heal)))
    (else
      (Ability|EndAbility))))

(event Ability|EventOnEndAbility
  (params (bWasCancelled :type Boolean))
  (|CallParentFunction bWasCancelled)
  (bind _botcontroller (Variables|InternalData|GetBotController)
    (:then
      (Class|FortAthenaAIBotController|EquipBestWeapon _botcontroller false))
    (:else)))

(fn CanActivateAbility
  (params (ActorInfo :type Gameplay%20Ability%20Actor%20Info%20Structure) (Handle :type Gameplay%20Ability%20Spec%20Handle%20Structure))
  (bind _asfort_player_pawn (Utilities|Casting|CastToFortPlayerPawn (Ability|GetAvatarActorfromActorInfo))
    (:then
      (return 0 (or (<= (/ (FortPawn|GetHealth _asfort_player_pawn) (FortPawn|GetMaxHealth _asfort_player_pawn)) (ScalableFloat|GetValueatLevel (Variables|Default|GetPawnHealthPercentage))) (<= (/ (FortPawn|GetShield _asfort_player_pawn) (FortPawn|GetMaxShield _asfort_player_pawn)) (ScalableFloat|GetValueatLevel (Variables|Default|GetPawnShieldPercentage))))))
    (:CastFailed
      (return 0 false))))
